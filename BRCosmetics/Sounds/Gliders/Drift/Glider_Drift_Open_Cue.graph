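# Sound cue for the opening sound of the Drift glider cosmetic.
# A mixer layers three waves at equal volume: one open layer plays immediately, a second follows after a short fixed delay, and the main open wave starts last, a second in.
# Two mixer inputs are left empty.
# The cue uses glider open/close attenuation and concurrency, a glider open property sheet, and the glider land sound class.
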SoundCue /BRCosmetics/Sounds/Gliders/Drift/Glider_Drift_Open_Cue
root=Mixer_0
nodes [
 {id=Delay_0, children [WavePlayer_2], properties {delayMax=1, delayMin=1}}
 {id=Delay_1, children [WavePlayer_10], properties {delayMax=0.2, delayMin=0.2}}
 {id=Mixer_0, children [Delay_0, WavePlayer_11, Delay_1, null, null], properties {inputVolume [1,1,1,1,1]}}
 {id=WavePlayer_10, properties {soundWaveAssetPtr=&/Glider_Drift_OpenLayer_08}}
 {id=WavePlayer_11, properties {soundWaveAssetPtr=&/Glider_Drift_OpenLayer_09}}
 {id=WavePlayer_2, properties {soundWaveAssetPtr=&/Glider_Drift_Open_01}}
]
properties {
 attenuationSettings=/Game/Sounds/Attenuation/Fort_Player_Glider_OpenClose_Athena_Att
 audioPropertiesSheet=/BRCosmetics/Sounds/Settings/PropertySheets/Glider/APS_Cosmetic_Glider_Open
 maxDistance=10000
 soundClassObject=&Fort_Gliders_Land
 volumeMultiplier=0.6
 concurrencySet [&Athena_GliderOpenClose_Conc]
}
